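# Animation Blueprint for the F_MED_Deco_01 female player body that follows a copied pose and adds corrective motion to the shoulder pad.
# On initialization it caches the owner skeletal mesh component as MeshToCopy, which feeds Copy Pose From Mesh.
# A Pose Driver reads upperarm_l against six RBF targets to drive only dyn_shoulderPad from the armor PoseAsset, then outputs the pose.
# The ROM sequence player is unconnected and the update event is empty.
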AnimBlueprint /Game/Characters/Player/Female/Medium/Bodies/F_MED_Deco_01/Meshes/F_MED_Deco_Body_01_AnimBP
target_skeleton=&/F_MED_Deco_Body_01_Skeleton
graphs [
 AnimGraph {nodes [
   n1 = Pose Driver - Source: upperarm_l!(SourcePose:Pose Link Structure=Copy Pose From Mesh(SourceMeshComponent=Get MeshToCopy() -> (MeshToCopy:Skeletal Mesh Component Object Reference))) -> (Pose:Pose Link Structure) {
    is=title AnimGraphNode_PoseDriver
    poseAsset=&/F_MED_Deco_Armor_PoseAsset
    onlyDriveBones [{boneName=dyn_shoulderPad}]
    poseTargets [
     {
      bApplyCustomCurve=false
      bIsHidden=false
      distanceMethod=DefaultMethod
      drivenName=Pose_0
      functionType=DefaultFunction
      targetScale=0.3153
      boneTransforms [{targetRotation=<rot -58.1985 2.4027 1.3901>, targetTranslation=<9.8658 -1.2136 -0.4105>}]
      customCurve {defaultValue=3.4028234663852886e+38, postInfinityExtrap=RCCE_Constant, preInfinityExtrap=RCCE_Constant, keys []}
      targetRotation=<rot 0 0 0>
     }
     {
      bApplyCustomCurve=false
      bIsHidden=false
      distanceMethod=DefaultMethod
      drivenName=Pose_1
      functionType=DefaultFunction
      targetScale=0.3153
      boneTransforms [{targetRotation=<rot -86.4965 -8.4704 12.0766>, targetTranslation=<9.8658 -1.2136 -0.4105>}]
      customCurve {defaultValue=3.4028234663852886e+38, postInfinityExtrap=RCCE_Constant, preInfinityExtrap=RCCE_Constant, keys []}
      targetRotation=<rot 0 0 0>
     }
     {
      bApplyCustomCurve=false
      bIsHidden=false
      distanceMethod=DefaultMethod
      drivenName=Pose_2
      functionType=DefaultFunction
      targetScale=0.6913
      boneTransforms [{targetRotation=<rot 3.9936 3.6354 0.7343>, targetTranslation=<9.8658 -1.2136 -0.4105>}]
      customCurve {defaultValue=3.4028234663852886e+38, postInfinityExtrap=RCCE_Constant, preInfinityExtrap=RCCE_Constant, keys []}
      targetRotation=<rot 0 0 0>
     }
     {
      bApplyCustomCurve=false
      bIsHidden=false
      distanceMethod=DefaultMethod
      drivenName=Pose_3
      functionType=DefaultFunction
      targetScale=0.9962
      boneTransforms [{targetRotation=<rot 86.2929 172.2027 168.5952>, targetTranslation=<9.8658 -1.2136 -0.4105>}]
      customCurve {defaultValue=3.4028234663852886e+38, postInfinityExtrap=RCCE_Constant, preInfinityExtrap=RCCE_Constant, keys []}
      targetRotation=<rot 0 0 0>
     }
     {
      bApplyCustomCurve=false
      bIsHidden=false
      distanceMethod=DefaultMethod
      drivenName=Pose_4
      functionType=DefaultFunction
      targetScale=0.9797
      boneTransforms [{targetRotation=<rot -0.6035 91.7417 4.0154>, targetTranslation=<9.8658 -1.2136 -0.4105>}]
      customCurve {defaultValue=3.4028234663852886e+38, postInfinityExtrap=RCCE_Constant, preInfinityExtrap=RCCE_Constant, keys []}
      targetRotation=<rot 0 0 0>
     }
     {
      bApplyCustomCurve=false
      bIsHidden=false
      distanceMethod=DefaultMethod
      drivenName=Pose_5
      functionType=DefaultFunction
      targetScale=1
      boneTransforms [{targetRotation=<rot 0.5783 -88.6179 -4.0191>, targetTranslation=<9.8658 -1.2136 -0.4105>}]
      customCurve {defaultValue=3.4028234663852886e+38, postInfinityExtrap=RCCE_Constant, preInfinityExtrap=RCCE_Constant, keys []}
      targetRotation=<rot 0 0 0>
     }
    ]
    rBFParams {radius=44.989}
    sourceBones [{boneName=upperarm_l}]
   }
   Output Pose(Result=@n1) {layerGroup=None}
   Sequence Player 'F_MED_Deco_ROM'() {sequence=&/F_MED_Deco_ROM}
  ]}
 EventGraph {nodes [
   Event BlueprintInitializeAnimation()
   > Set MeshToCopy(MeshToCopy:Skeletal Mesh Component Object Reference=GetOwnerSkeletalMeshComponent()) -> (Output_Get:Skeletal Mesh Component Object Reference)
   Event BlueprintUpdateAnimation()
  ]}
]
variables [MeshToCopy {type=SkeletalMeshComponent}]
